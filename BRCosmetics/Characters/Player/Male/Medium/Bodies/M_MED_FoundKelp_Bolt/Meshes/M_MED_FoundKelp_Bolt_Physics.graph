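# Physics asset for the M_MED_FoundKelp_Bolt male medium body cosmetic, set up for secondary motion on its jacket.
# Small capsules cover front, side and back jacket bones on both sides as Base/dyn pairs, while large capsules cover pelvis and spine_03.
# Each dyn bone is constrained to its Base bone with tight swing and near-zero twist limits.
# Base bones link to spine_03, and spine_03 to pelvis, through free angular constraints.
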PhysicsAsset /BRCosmetics/Characters/Player/Male/Medium/Bodies/M_MED_FoundKelp_Bolt/Meshes/M_MED_FoundKelp_Bolt_Physics
bodies {
 Base_JacketB_l [{shapeType=Capsule, length=2, radius=1, shapeName=Base_JacketB_l_capsule, center=<2 0 0>, rotation=<rot 90 0 0>}]
 Base_JacketB_r [{shapeType=Capsule, length=2, radius=1, shapeName=Base_JacketB_r_capsule, center=<2 0 0>, rotation=<rot 90 0 0>}]
 Base_JacketF_l [{shapeType=Capsule, length=2, radius=1, shapeName=Base_JacketF_l_capsule, rotation=<rot 90 0 0>}]
 Base_JacketF_r [{shapeType=Capsule, length=0.505, radius=0.505, shapeName=Base_JacketF_r_capsule, center=<2 0 0>, rotation=<rot 90 0 0>}]
 Base_JacketS_l [{shapeType=Capsule, length=0.505, radius=0.505, shapeName=Base_JacketS_l_capsule, center=<2 0 0>, rotation=<rot 90 0 0>}]
 Base_JacketS_r1 [{shapeType=Capsule, length=0.505, radius=0.505, shapeName=Base_JacketS_r1_capsule, center=<2 0 0>, rotation=<rot 90 0 0>}]
 dyn_JacketB_l [{shapeType=Capsule, length=2, radius=1, shapeName=dyn_JacketB_l_capsule, center=<2 0 0>, rotation=<rot 90 0 0>}]
 dyn_JacketB_r [{shapeType=Capsule, length=2, radius=1, shapeName=dyn_JacketB_r_capsule, center=<2 0 0>, rotation=<rot 90 0 0>}]
 dyn_JacketF_l [{shapeType=Capsule, length=2, radius=1, shapeName=dyn_JacketF_l_capsule, rotation=<rot 90 0 0>}]
 dyn_JacketF_r [{shapeType=Capsule, length=2, radius=1, shapeName=dyn_JacketF_r_capsule, center=<2 0 0>, rotation=<rot 90 0 0>}]
 dyn_JacketS_l [{shapeType=Capsule, length=2, radius=1, shapeName=dyn_JacketS_l_capsule, center=<2 0 0>, rotation=<rot 90 0 0>}]
 dyn_JacketS_r [{shapeType=Capsule, length=2, radius=1, shapeName=dyn_JacketS_r_capsule, center=<2 0 0>, rotation=<rot 90 0 0>}]
 pelvis [{
   shapeType=Capsule
   length=18.61
   radius=14.4831
   shapeName=pelvis_capsule
   center=<1.6289 -2.1081 1.3651>
   rotation=<rot -14.4221 -90 -1.3547>
  }]
 spine_03 [{
   shapeType=Capsule
   length=18.0433
   radius=16.2384
   shapeName=spine_03_capsule
   center=<10.2683 0.9111 0.3132>
   rotation=<rot 0 3.3849 -174.5447>
  }]
}
constraints [
 {bone1Name=dyn_JacketB_l, bone2Name=Base_JacketB_l, swing1LimitDegrees=10, swing2LimitDegrees=10, twistLimitDegrees=1}
 {bone1Name=dyn_JacketB_r, bone2Name=Base_JacketB_r, swing1LimitDegrees=10, swing2LimitDegrees=10, twistLimitDegrees=1}
 {bone1Name=dyn_JacketF_l, bone2Name=Base_JacketF_l, swing1LimitDegrees=10, swing2LimitDegrees=10, twistLimitDegrees=1}
 {bone1Name=dyn_JacketF_r, bone2Name=Base_JacketF_r, swing1LimitDegrees=10, swing2LimitDegrees=10, twistLimitDegrees=1}
 {bone1Name=dyn_JacketS_l, bone2Name=Base_JacketS_l, swing1LimitDegrees=10, swing2LimitDegrees=10, twistLimitDegrees=1}
 {
  bone1Name=Base_JacketF_l
  bone2Name=spine_03
  swing1LimitDegrees=45
  swing1Motion=Free
  swing2LimitDegrees=45
  swing2Motion=Free
  twistLimitDegrees=45
  twistMotion=Free
 }
 {
  bone1Name=Base_JacketS_l
  bone2Name=spine_03
  swing1LimitDegrees=45
  swing1Motion=Free
  swing2LimitDegrees=45
  swing2Motion=Free
  twistLimitDegrees=45
  twistMotion=Free
 }
 {
  bone1Name=Base_JacketB_l
  bone2Name=spine_03
  swing1LimitDegrees=45
  swing1Motion=Free
  swing2LimitDegrees=45
  swing2Motion=Free
  twistLimitDegrees=45
  twistMotion=Free
 }
 {
  bone1Name=Base_JacketF_r
  bone2Name=spine_03
  swing1LimitDegrees=45
  swing1Motion=Free
  swing2LimitDegrees=45
  swing2Motion=Free
  twistLimitDegrees=45
  twistMotion=Free
 }
 {
  bone1Name=Base_JacketB_r
  bone2Name=spine_03
  swing1LimitDegrees=45
  swing1Motion=Free
  swing2LimitDegrees=45
  swing2Motion=Free
  twistLimitDegrees=45
  twistMotion=Free
 }
 {
  bone1Name=spine_03
  bone2Name=pelvis
  swing1LimitDegrees=45
  swing1Motion=Free
  swing2LimitDegrees=45
  swing2Motion=Free
  twistLimitDegrees=45
  twistMotion=Free
 }
 {
  bone1Name=Base_JacketS_r1
  bone2Name=spine_03
  swing1LimitDegrees=45
  swing1Motion=Free
  swing2LimitDegrees=45
  swing2Motion=Free
  twistLimitDegrees=45
  twistMotion=Free
 }
 {bone1Name=dyn_JacketS_r, bone2Name=Base_JacketS_r1, swing1LimitDegrees=10, swing2LimitDegrees=10, twistLimitDegrees=1}
]
