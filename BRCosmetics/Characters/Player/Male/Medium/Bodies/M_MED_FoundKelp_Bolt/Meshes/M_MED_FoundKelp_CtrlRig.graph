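# Control Rig for the M_MED_FoundKelp_Bolt player body that runs a shared deform rig, then procedurally deforms the neck collar bones.
# Weighted parent constraints attach the collar to spine_05 and the clavicles.
# Head, neck and clavicle rotations, measured against construction-time reference controls with swing/twist and remaps, weight side, back and twist corrective poses, with IsHemet selecting the pose set.
# Clavicle rotations update after use, lagging one evaluation.
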ControlRigBlueprint /BRCosmetics/Characters/Player/Male/Medium/Bodies/M_MED_FoundKelp_Bolt/Meshes/M_MED_FoundKelp_CtrlRig
graphs [
 Add_1_ContainedGraph {nodes [
   n1 = RigVMFunctionEntryNode()
   Add = &RigVMFunction_MathDoubleAdd(A=@n1.A, B=@n1.B)
   RigVMFunctionReturnNode(Result=&RigVMFunction_MathDoubleAdd(A=@Add.Result, B=@n1.C).Result)
  ]}
 Maximum_2_ContainedGraph {nodes [
   n1 = RigVMFunctionEntryNode()
   RigVMFunctionReturnNode(Result=&RigVMFunction_MathDoubleMax(B=@n1.C, A=&RigVMFunction_MathDoubleMax(A=@n1.A, B=@n1.B).Result).Result)
  ]}
 Maximum_2_ContainedGraph {nodes [
   n1 = RigVMFunctionEntryNode()
   RigVMFunctionReturnNode(Result=&RigVMFunction_MathDoubleMax(B=@n1.C, A=&RigVMFunction_MathDoubleMax(A=@n1.A, B=@n1.B).Result).Result)
  ]}
 Neck Collar Constraints_ContainedGraph {nodes [
   RigVMFunctionEntryNode()
   > &RigUnit_ParentConstraint(
    AdvancedSettings=(InterpolationType=Average,RotationOrderForFilter=XZY)
    Child=(Type=Bone,Name="collar_back_m")
    Filter=(TranslationFilter=(bX=true,bY=true,bZ=true),RotationFilter=(bX=True,bY=True,bZ=True),ScaleFilter=(bX=True,bY=True,bZ=True))
    Parents=((Item=(Type=Bone,Name="spine_05"),Weight=0.8),(Item=(Type=Bone,Name="clavicle_l"),Weight=0.1),(Item=(Type=Bone,Name="clavicle_r"),Weight=0.1))
    Weight=1
   )
   > &RigUnit_ParentConstraint(
    AdvancedSettings=(InterpolationType=Average,RotationOrderForFilter=XZY)
    Child=(Type=Bone,Name="collar_front_l")
    Filter=(TranslationFilter=(bX=true,bY=true,bZ=true),RotationFilter=(bX=True,bY=True,bZ=True),ScaleFilter=(bX=True,bY=True,bZ=True))
    Parents=((Item=(Type=Bone,Name="spine_05"),Weight=0.7),(Item=(Type=Bone,Name="clavicle_l"),Weight=0.3))
    Weight=1
   )
   > &RigUnit_ParentConstraint(
    AdvancedSettings=(InterpolationType=Average,RotationOrderForFilter=XZY)
    Child=(Type=Bone,Name="collar_side_l")
    Filter=(TranslationFilter=(bX=true,bY=true,bZ=true),RotationFilter=(bX=True,bY=True,bZ=True),ScaleFilter=(bX=True,bY=True,bZ=True))
    Parents=((Item=(Type=Bone,Name="spine_05"),Weight=0.7),(Item=(Type=Bone,Name="clavicle_l"),Weight=0.3))
    Weight=1
   )
   > &RigUnit_ParentConstraint(
    AdvancedSettings=(InterpolationType=Average,RotationOrderForFilter=XZY)
    Child=(Type=Bone,Name="collar_back_l")
    Filter=(TranslationFilter=(bX=true,bY=true,bZ=true),RotationFilter=(bX=True,bY=True,bZ=True),ScaleFilter=(bX=True,bY=True,bZ=True))
    Parents=((Item=(Type=Bone,Name="spine_05"),Weight=0.75),(Item=(Type=Bone,Name="clavicle_l"),Weight=0.15))
    Weight=1
   )
   > &RigUnit_ParentConstraint(
    AdvancedSettings=(InterpolationType=Average,RotationOrderForFilter=XZY)
    Child=(Type=Bone,Name="collar_front_r")
    Filter=(TranslationFilter=(bX=true,bY=true,bZ=true),RotationFilter=(bX=True,bY=True,bZ=True),ScaleFilter=(bX=True,bY=True,bZ=True))
    Parents=((Item=(Type=Bone,Name="spine_05"),Weight=0.7),(Item=(Type=Bone,Name="clavicle_r"),Weight=0.3))
    Weight=1
   )
   > &RigUnit_ParentConstraint(
    AdvancedSettings=(InterpolationType=Average,RotationOrderForFilter=XZY)
    Child=(Type=Bone,Name="collar_side_r")
    Filter=(TranslationFilter=(bX=true,bY=true,bZ=true),RotationFilter=(bX=True,bY=True,bZ=True),ScaleFilter=(bX=True,bY=True,bZ=True))
    Parents=((Item=(Type=Bone,Name="spine_05"),Weight=0.7),(Item=(Type=Bone,Name="clavicle_r"),Weight=0.3))
    Weight=1
   )
   > &RigUnit_ParentConstraint(
    AdvancedSettings=(InterpolationType=Average,RotationOrderForFilter=XZY)
    Child=(Type=Bone,Name="collar_back_r")
    Filter=(TranslationFilter=(bX=true,bY=true,bZ=true),RotationFilter=(bX=True,bY=True,bZ=True),ScaleFilter=(bX=True,bY=True,bZ=True))
    Parents=((Item=(Type=Bone,Name="spine_05"),Weight=0.75),(Item=(Type=Bone,Name="clavicle_r"),Weight=0.15))
    Weight=1
   )
   > RigVMFunctionReturnNode()
  ]}
 New Function_ContainedGraph {nodes [
   RigVMFunctionEntryNode()
   n1 = RigVMDispatchNode(Condition=false, False=(X=-70,Y=34,Z=50), True=(X=-70,Y=60,Z=20)) {template="DISPATCH_RigVMDispatch_If(in Condition,in True,in False,out Result)"}
   n2 = &RigVMFunction_MathQuaternionSwingTwist(
    TwistAxis=(X=1,Y=0,Z=0)
    Input=&RigUnit_GetRelativeTransformForItem(Child=(Type=Bone,Name="head"), Parent=(Type=Control,Name="ref_head_ctrl"), bChildInitial=False, bParentInitial=False).RelativeTransform.Rotation
   )
   RigVMFunction_MathQuaternionToEuler = &RigVMFunction_MathQuaternionToEuler(RotationOrder=ZYX, Value=@n2.Swing)
   n3 = &RigVMFunction_MathQuaternionToEuler(
    RotationOrder=ZYX
    Value=&RigVMFunction_MathQuaternionSwingTwist(
     TwistAxis=(X=1,Y=0,Z=0)
     Input=&RigUnit_GetRelativeTransformForItem(Child=(Type=Bone,Name="neck_02"), Parent=(Type=Control,Name="ref_neck_ctrl"), bChildInitial=False, bParentInitial=False).RelativeTransform.Rotation
    ).Swing
   )
   Maximum_2 = RigVMAggregateNode(
    B=&RigVMFunction_MathDoubleRemap(SourceMinimum=0, TargetMaximum=1, TargetMinimum=0, SourceMaximum=@n1.Result.Y, Value=@RigVMFunction_MathQuaternionToEuler.Result.Y).Result
    C=&RigVMFunction_MathDoubleRemap(SourceMinimum=0, TargetMaximum=1, TargetMinimum=0, SourceMaximum=@n1.Result.Z, Value=@n3.Result.Y).Result
    A=&RigVMFunction_MathDoubleRemap(SourceMinimum=0, TargetMaximum=1, TargetMinimum=0, SourceMaximum=@n1.Result.X, Value=RigVMVariableNode() {variable=Clavicle_rot_l}.Value.Y).Result
   )
   > RigVMVariableNode(Value=@Maximum_2.Result) {variable=head_side_l}
   n4 = &RigVMFunction_MathQuaternionToEuler(RotationOrder=ZYX, Value=@n2.Twist)
   > RigVMVariableNode(Value=/Script/RigVM.RigVMFunction_AnimEasing(
     SourceMaximum=1
     SourceMinimum=0
     TargetMaximum=1
     TargetMinimum=0
     Type=CubicEaseOut
     Value=&RigVMFunction_MathDoubleRemap(SourceMaximum=-50, SourceMinimum=-10, TargetMaximum=1, TargetMinimum=0, Value=@n4.Result.X).Result
    ) {is=struct RigVMUnitNode}.Result) {variable=head_twist_l}
   n5 = RigVMDispatchNode(Condition=False, False=(X=-70,Y=-70,Z=-20), True=(X=-70,Y=-60,Z=-20)) {template="DISPATCH_RigVMDispatch_If(in Condition,in True,in False,out Result)"}
   n6 = &RigVMFunction_MathDoubleRemap(SourceMinimum=0, TargetMaximum=1, TargetMinimum=0, SourceMaximum=@n5.Result.X, Value=RigVMVariableNode() {variable=Clavicle_rot_r}.Value.Y)
   Add = RigVMAggregateNode(
    B=&RigVMFunction_MathDoubleRemap(SourceMinimum=0, TargetMaximum=1, TargetMinimum=0, SourceMaximum=@n5.Result.Y, Value=@RigVMFunction_MathQuaternionToEuler.Result.Y).Result
    C=&RigVMFunction_MathDoubleRemap(SourceMinimum=0, TargetMaximum=1, TargetMinimum=0, SourceMaximum=@n5.Result.Z, Value=@n3.Result.Y).Result
    A=@n6.Result
   )
   Maximum_1 = RigVMAggregateNode(B=@Add.Result, A=@n6.Result)
   > RigVMVariableNode(Value=@Maximum_1.Result) {variable=head_side_r}
   > RigVMVariableNode(Value=&RigVMFunction_MathDoubleRemap(SourceMaximum=50, SourceMinimum=34, TargetMaximum=1, TargetMinimum=0, Value=@n4.Result.X).Result) {variable=head_twist_r}
   RigVMFunction_AnimEasing = &RigVMFunction_AnimEasing(
    SourceMaximum=1
    SourceMinimum=0
    TargetMaximum=1
    TargetMinimum=0
    Type=CubicEaseIn
    Value=&RigVMFunction_MathDoubleRemap(SourceMaximum=-40, SourceMinimum=0, TargetMaximum=0.85, TargetMinimum=0, Value=@RigVMFunction_MathQuaternionToEuler.Result.Z).Result
   )
   > RigVMVariableNode(Value=&RigVMFunction_MathDoubleAdd(
     B=&RigVMFunction_MathDoubleRemap(SourceMaximum=-40, SourceMinimum=0, TargetMaximum=1, TargetMinimum=0, Value=@n3.Result.Z).Result
     A=@RigVMFunction_AnimEasing.Result
    ).Result) {variable=head_back_m}
   > RigVMFunctionReturnNode()
   Cast = &RigVMFunction_MathBoolToFloat(Value=False)
   &RigVMFunction_MathBoolToFloat(Value=False)
   RigVMDispatchNode(Condition=False, False=(X=-80,Y=-55,Z=0), True=(X=-50,Y=-70,Z=0)) {template="DISPATCH_RigVMDispatch_If(in Condition,in True,in False,out Result)"}
  ]}
 RigVMFunctionLibrary {nodes [
   Neck Collar Constraints = RigVMCollapseNode(ExecuteContext=())
   "Neck Collar Variables " = RigVMCollapseNode(ExecuteContext=())
  ]}
 RigVMFunction_Sequence_1_ContainedGraph {nodes [
   n1 = RigVMFunctionEntryNode()
   RigVMFunction_Sequence = &RigVMFunction_Sequence(ExecuteContext=@n1.ExecuteContext)
   n2 = &RigVMFunction_Sequence(ExecuteContext=@RigVMFunction_Sequence.B)
   RigVMFunctionReturnNode(A=@RigVMFunction_Sequence.A, B=@n2.A, C=@n2.B)
  ]}
 RigVMModel {nodes [
   RigUnit_BeginExecution = &RigUnit_BeginExecution()
   FN_DeformRig_FNC = RigVMFunctionReferenceNode(
    DeformRig_ Curve="(deformRig_l_upperarm_up_alpha_72_#1=False,deformRig_r_upperarm_up_alpha_73_#2=False,deformRig_l_upperarm_fr_alpha_74_#3=False,deformRig_r_upperarm_fr_alpha_75_CB51B8 ...(truncated)"
    DeformRig_Bool="(upperarm_1_#4=true,elbow_3_#5=true,wrist_7_#6=true,groin_8_#7=true,glute_17_#8=true,knee_9_40C408324CA261 ...(truncated)"
    DeformRig_Debug=(upperarm_debug_18_#4=False,wrist_debug_20_#6=False,lat_debug_26_#9=False,groinGlute_debug_27_#7=False)
    DeformRig_Value="(upperarm_up_18_#10=(Rotation=(X=0.0,Y=-0.0,Z=0.0,W=1.0),Translation=(X=0.0,Y=0.0,Z=8.0),Scale3D=(X=0.70,Y=1.0,Z=1.0)),upperarm_fwd_fr_19_#11=(Rotation=(X=0.0,Y=-0.0,Z=0.0,W=1.0),Translation=(X=0.0, ...(truncated)"
    ExecuteContext=@RigUnit_BeginExecution.ExecutePin
   )
   > RigVMFunction_Sequence = RigVMAggregateNode()
   Neck Collar Constraints = RigVMFunctionReferenceNode(ExecuteContext=@RigVMFunction_Sequence.A)
   > "Neck Collar Variables " = RigVMFunctionReferenceNode()
   RigVMFunction_ControlFlowBranch = &RigVMFunction_ControlFlowBranch(ExecuteContext=@"Neck Collar Variables ".ExecuteContext, Condition=RigVMVariableNode() {variable=IsHemet}.Value)
   &RigUnit_ModifyTransforms(
    ItemToModify="((Item=(Type=Bone,Name=\"collar_side_l\"),Transform=(Rotation=(X=0,Y=-0.456,Z=0,W=0.89),Translation=(X=-1.3785,Y=-0.2747,Z=-1.915),Scale3D=(X=1,Y=1,Z=1))),(Item=(Type=Bone,Name=\"collar_front_l\"),Transform=(Rotation=(X=0,Y=0.6682,Z=0,W=-0.744),Translation=(X= ...(truncated)"
    WeightMinimum=0
    ExecutePin=@RigVMFunction_ControlFlowBranch.False
    Weight=RigVMVariableNode() {variable=head_side_l}.Value
   )
   > &RigUnit_ModifyTransforms(
    ItemToModify="((Item=(Type=Bone,Name=\"collar_side_r\"),Transform=(Rotation=(X=-0.1038,Y=-0.6696,Z=-0.0107,W=0.7353),Translation=(X=0.2157,Y=1.0083,Z=0.5336),Scale3D=(X=1,Y=1,Z=1))),(Item=(Type=Bone,Name=\"collar_front_r\"),Transform=(Rotation=(X=-0.121,Y=-0.4147,Z=0.3715,W ...(truncated)"
    WeightMinimum=0
    Weight=RigVMVariableNode() {variable=head_side_r}.Value
   )
   > &RigUnit_ModifyTransforms(
    ItemToModify="((Item=(Type=Bone,Name=\"collar_back_m\"),Transform=(Rotation=(X=0,Y=-0.5639,Z=0,W=0.8259),Translation=(X=-0.3376,Y=0,Z=-4.6865),Scale3D=(X=1,Y=1,Z=1))),(Item=(Type=Bone,Name=\"collar_back_r\"),Transform=(Rotation=(X=0,Y=0.887,Z=0,W=-0.4617),Translation=(X=-0. ...(truncated)"
    WeightMinimum=0
    Weight=RigVMVariableNode() {variable=head_back_m}.Value
   )
   > &RigUnit_ModifyTransforms(
    ItemToModify="((Item=(Type=Bone,Name=\"collar_side_l\"),Transform=(Rotation=(X=0,Y=0.36,Z=0,W=-0.9329),Translation=(X=-0.4368,Y=0,Z=-1.5375),Scale3D=(X=1,Y=1,Z=1))),(Item=(Type=Bone,Name=\"collar_front_l\"),Transform=(Rotation=(X=0,Y=-0.3118,Z=0,W=0.9501),Translation=(X=-1. ...(truncated)"
    WeightMinimum=0
    Weight=RigVMVariableNode() {variable=head_twist_l}.Value
   )
   > &RigUnit_ModifyTransforms(
    ItemToModify="((Item=(Type=Bone,Name=\"collar_side_r\"),Transform=(Rotation=(X=0,Y=0.4927,Z=0,W=-0.8702),Translation=(X=0.5205,Y=0,Z=1.3998),Scale3D=(X=1,Y=1,Z=1))),(Item=(Type=Bone,Name=\"collar_front_r\"),Transform=(Rotation=(X=0,Y=0.6418,Z=0,W=-0.7668),Translation=(X=0,Y ...(truncated)"
    WeightMinimum=0
    Weight=RigVMVariableNode() {variable=head_twist_r}.Value
   )
   &RigUnit_ModifyTransforms(
    ItemToModify="((Item=(Type=Bone,Name=\"collar_side_l\"),Transform=(Rotation=(X=0.1118,Y=0.4523,Z=0.0573,W=-0.8829),Translation=(X=-1.3786,Y=-0.2747,Z=-1.915),Scale3D=(X=1,Y=1,Z=1))),(Item=(Type=Bone,Name=\"collar_front_l\"),Transform=(Rotation=(X=0,Y=0.6682,Z=0,W=-0.744),Tr ...(truncated)"
    WeightMinimum=0
    ExecutePin=@RigVMFunction_ControlFlowBranch.True
    Weight=RigVMVariableNode() {variable=head_side_l}.Value
   )
   > &RigUnit_ModifyTransforms(
    ItemToModify="((Item=(Type=Bone,Name=\"collar_side_r\"),Transform=(Rotation=(X=-0.1038,Y=-0.6696,Z=-0.0107,W=0.7353),Translation=(X=0.2157,Y=1.0083,Z=0.5336),Scale3D=(X=1,Y=1,Z=1))),(Item=(Type=Bone,Name=\"collar_front_r\"),Transform=(Rotation=(X=-0.121,Y=-0.4147,Z=0.3715,W ...(truncated)"
    WeightMinimum=0
    Weight=RigVMVariableNode() {variable=head_side_r}.Value
   )
   > &RigUnit_ModifyTransforms(
    ItemToModify="((Item=(Type=Bone,Name=\"collar_back_m\"),Transform=(Rotation=(X=0.0204,Y=-0.7257,Z=0.01,W=0.6876),Translation=(X=-0.1253,Y=0.0003,Z=-4.1435),Scale3D=(X=1,Y=1,Z=1))),(Item=(Type=Bone,Name=\"collar_back_r\"),Transform=(Rotation=(X=-0.1912,Y=-0.7287,Z=0.3832,W=0 ...(truncated)"
    WeightMinimum=0
    Weight=RigVMVariableNode() {variable=head_back_m}.Value
   )
   > &RigUnit_ModifyTransforms(
    ItemToModify="((Item=(Type=Bone,Name=\"collar_side_l\"),Transform=(Rotation=(X=0,Y=0.0615,Z=0,W=-0.9981),Translation=(X=0,Y=0,Z=0),Scale3D=(X=1,Y=1,Z=1))),(Item=(Type=Bone,Name=\"collar_front_l\"),Transform=(Rotation=(X=0.022,Y=-0.6125,Z=-0.0844,W=0.7856),Translation=(X=-0. ...(truncated)"
    WeightMinimum=0
    Weight=RigVMVariableNode() {variable=head_twist_l}.Value
   )
   > &RigUnit_ModifyTransforms(
    ItemToModify="((Item=(Type=Bone,Name=\"collar_side_r\"),Transform=(Rotation=(X=0,Y=0.4927,Z=0,W=-0.8702),Translation=(X=0.5205,Y=0,Z=1.3998),Scale3D=(X=1,Y=1,Z=1))),(Item=(Type=Bone,Name=\"collar_front_r\"),Transform=(Rotation=(X=0,Y=0.6418,Z=0,W=-0.7668),Translation=(X=0,Y ...(truncated)"
    WeightMinimum=0
    Weight=RigVMVariableNode() {variable=head_twist_r}.Value
   )
   n1 = &RigUnit_GetRelativeTransformForItem(Child=(Type=Bone,Name="clavicle_l"), Parent=(Type=Control,Name="ref_clavicle_l_ctrl"), bChildInitial=False, bParentInitial=False)
   RigVMVariableNode(
    Value=&RigVMFunction_MathQuaternionToEuler(RotationOrder=ZYX, Value=@n1.RelativeTransform.Rotation).Result
    ExecuteContext=@RigVMFunction_Sequence.B
   ) {variable=Clavicle_rot_l}
   n2 = &RigUnit_GetRelativeTransformForItem(Child=(Type=Bone,Name="clavicle_r"), Parent=(Type=Control,Name="ref_clavicle_r_ctrl"), bChildInitial=False, bParentInitial=False)
   RigVMVariableNode(
    Value=&RigVMFunction_MathQuaternionToEuler(RotationOrder=ZYX, Value=@n2.RelativeTransform.Rotation).Result
    ExecuteContext=@RigVMFunction_Sequence.C
   ) {variable=Clavicle_rot_r}
   setFKCtrl = RigVMFunctionReferenceNode(ExecuteContext=(), setFK_Ctrl=true)
  ]}
 RigVMModel Construction Event Graph {nodes [
   n1 = &RigUnit_PrepareForExecution()
   &RigUnit_HierarchyImportFromSkeleton(NameSpace=None, bIncludeCurves=False, bIncludeMeshSockets=False, bIncludeVirtualBones=True, ExecutePin=@n1.ExecutePin)
   > Control Stack at Item = RigVMFunctionReferenceNode(
    Add Bottom Null=False
    Add Null=true
    Add Secondary Control=False
    Control Suffix=_ctrl
    Item=(Name="head",Type=Bone)
    Name=ref_head
    Null Suffix=_null
    Parent=(Name="neck_02",Type=Bone)
    Secondary Settings="(Shape=(bVisible=True,Name=\"Default\",Color=(R=1,G=0,B=0,A=1),Transform=(Rotation=(X=0,Y=0,Z=0,W=1),Translation=(X=0,Y=0,Z=0),Scale3D=(X=1,Y=1,Z=1))),Proxy=(bIsProxy=False,ShapeVisibility=BasedOnSelection),DisplayName=\"None\",InitialSpace=LocalSpace,bUsePref ...(truncated)"
    Settings="(Shape=(bVisible=True,Name=\"Pyramid_Thin\",Color=(R=1,G=0,B=0,A=1),Transform=(Rotation=(X=0,Y=0,Z=0,W=1),Translation=(X=0,Y=0,Z=0),Scale3D=(X=1,Y=1,Z=1))),Proxy=(bIsProxy=False,ShapeVisibility=BasedOnSelection),DisplayName=\"None\",InitialSpace=LocalSpace,bUs ...(truncated)"
   )
   > Control Stack at Item_1 = RigVMFunctionReferenceNode(
    Add Bottom Null=False
    Add Null=true
    Add Secondary Control=False
    Control Suffix=_ctrl
    Item=(Name="neck_02",Type=Bone)
    Name=ref_neck
    Null Suffix=_null
    Parent=(Name="spine_05",Type=Bone)
    Secondary Settings="(Shape=(bVisible=True,Name=\"Default\",Color=(R=1,G=0,B=0,A=1),Transform=(Rotation=(X=0,Y=0,Z=0,W=1),Translation=(X=0,Y=0,Z=0),Scale3D=(X=1,Y=1,Z=1))),Proxy=(bIsProxy=False,ShapeVisibility=BasedOnSelection),DisplayName=\"None\",InitialSpace=LocalSpace,bUsePref ...(truncated)"
    Settings="(Shape=(bVisible=True,Name=\"Pyramid_Thin\",Color=(R=1,G=0,B=0,A=1),Transform=(Rotation=(X=0,Y=0,Z=0,W=1),Translation=(X=0,Y=0,Z=0),Scale3D=(X=1,Y=1,Z=1))),Proxy=(bIsProxy=False,ShapeVisibility=BasedOnSelection),DisplayName=\"None\",InitialSpace=LocalSpace,bUs ...(truncated)"
   )
   > Control Stack at Item_2 = RigVMFunctionReferenceNode(
    Add Bottom Null=False
    Add Null=true
    Add Secondary Control=False
    Control Suffix=_ctrl
    Item=(Name="clavicle_l",Type=Bone)
    Name=ref_clavicle_l
    Null Suffix=_null
    Parent=(Name="spine_05",Type=Bone)
    Secondary Settings="(Shape=(bVisible=True,Name=\"Default\",Color=(R=1,G=0,B=0,A=1),Transform=(Rotation=(X=0,Y=0,Z=0,W=1),Translation=(X=0,Y=0,Z=0),Scale3D=(X=1,Y=1,Z=1))),Proxy=(bIsProxy=False,ShapeVisibility=BasedOnSelection),DisplayName=\"None\",InitialSpace=LocalSpace,bUsePref ...(truncated)"
    Settings="(Shape=(bVisible=True,Name=\"Pyramid_Thin\",Color=(R=1,G=0,B=0,A=1),Transform=(Rotation=(X=0,Y=0,Z=0,W=1),Translation=(X=0,Y=0,Z=0),Scale3D=(X=1,Y=1,Z=1))),Proxy=(bIsProxy=False,ShapeVisibility=BasedOnSelection),DisplayName=\"None\",InitialSpace=LocalSpace,bUs ...(truncated)"
   )
   > Control Stack at Item_3 = RigVMFunctionReferenceNode(
    Add Bottom Null=False
    Add Null=true
    Add Secondary Control=False
    Control Suffix=_ctrl
    Item=(Name="clavicle_r",Type=Bone)
    Name=ref_clavicle_r
    Null Suffix=_null
    Parent=(Name="spine_05",Type=Bone)
    Secondary Settings="(Shape=(bVisible=True,Name=\"Default\",Color=(R=1,G=0,B=0,A=1),Transform=(Rotation=(X=0,Y=0,Z=0,W=1),Translation=(X=0,Y=0,Z=0),Scale3D=(X=1,Y=1,Z=1))),Proxy=(bIsProxy=False,ShapeVisibility=BasedOnSelection),DisplayName=\"None\",InitialSpace=LocalSpace,bUsePref ...(truncated)"
    Settings="(Shape=(bVisible=True,Name=\"Pyramid_Thin\",Color=(R=1,G=0,B=0,A=1),Transform=(Rotation=(X=0,Y=0,Z=0,W=1),Translation=(X=0,Y=0,Z=0),Scale3D=(X=1,Y=1,Z=1))),Proxy=(bIsProxy=False,ShapeVisibility=BasedOnSelection),DisplayName=\"None\",InitialSpace=LocalSpace,bUs ...(truncated)"
   )
   addFKCtrl = RigVMFunctionReferenceNode(
    Custom_FKCtrl=()
    ExecuteContext=()
    Options_FKCtrl="(pelvis_1_#12=false,spine_3_#13=false,clavicle_5_#14=false,arm_7_#15=false,leg_9_#16=false,head_11_46DC15EF465C ...(truncated)"
    Shape_Settings="(pelvis_2_#17=(bVisible=True,Name=\"Hexagon_Thin\",Color=(R=0.0117,G=0.8281,B=1,A=1),Transform=(Rotation=(X=0,Y=-0.7071,Z=0,W=0.7071),Translation=(X=2,Y=0,Z=0),Scale3D=(X=5,Y=5,Z=7))),spine_4_#18=(bVi ...(truncated)"
   )
  ]}
]
variables [
 Clavicle_rot_l {type=FVector, default=(X=0,Y=0,Z=0)}
 Clavicle_rot_r {type=FVector, default=(X=0,Y=0,Z=0)}
 IsHemet {type=bool, default=False}
 head_back_m {type=double, default=0}
 head_side_l {type=double, default=0}
 head_side_r {type=double, default=0}
 head_twist_l {type=double, default=0}
 head_twist_r {type=double, default=0}
]
