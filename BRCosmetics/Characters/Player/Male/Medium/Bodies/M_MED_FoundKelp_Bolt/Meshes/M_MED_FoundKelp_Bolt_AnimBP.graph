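# Post-process-style AnimBlueprint for the FoundKelp Bolt cosmetic character body, adding secondary motion to a pose copied from the attached parent.
# A Control Rig feeds RigidBody simulation, AnimDynamics on strap and can bones, then an additive Base_Strap offset.
# On customization load, the particle variant channel sets the rig's helmet flag.
# Each update computes state-based gravity into a variable, though the physics nodes show fixed gravity.
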AnimBlueprint /BRCosmetics/Characters/Player/Male/Medium/Bodies/M_MED_FoundKelp_Bolt/Meshes/M_MED_FoundKelp_Bolt_AnimBP
target_skeleton=&/M_MED_FoundKelp_Bolt_Skeleton
graphs [
 AnimGraph {nodes [
   n1 = Control Rig!(Source=Copy Pose From Mesh() {bUseAttachedParent=true}, Alpha, IsHemet:bool=Get isUsingHelmet() -> (isUsingHelmet:bool)) -> (Pose) {controlRigAssetReference {blueprintRigClass=&/M_MED_FoundKelp_CtrlRig.M_MED_FoundKelp_CtrlRig_C}}
   n2 = RigidBody!(ComponentPose=Local To Component(LocalPose=@n1), bAlphaBoolEnabled:bool, OverrideWorldGravity:Vector="0,0,-980", ExternalForce) -> (Pose) {
    alphaInputType=Bool
    bClampLinearTranslationLimitToRefPose=true
    bOverrideWorldGravity=true
    overridePhysicsAsset=&/M_MED_FoundKelp_Bolt_Physics
    simulationSpace=BaseBoneSpace
    alphaBoolBlend {blendOption=Linear}
    baseBoneRef {boneName=spine_04}
    componentAppliedLinearAccClamp=<0 0 0>
    overrideWorldGravity {z=-980}
    simSpaceSettings {maxAngularAcceleration=1, maxAngularVelocity=3, maxLinearAcceleration=800, maxLinearVelocity=500, worldAlpha=1, externalLinearDragV=<1 1 1>}
   }
   n3 = Anim Dynamics - Bone: dyn_Strap!(ComponentPose=@n2, Alpha=1, GravityOverride:Vector="0,0,-980", ExternalForce="0,0,0") -> (Pose) {
    angularDampingOverride=1
    angularSpringConstant=1000
    bAngularSpring=true
    bOverrideAngularDamping=true
    bUseGravityOverride=true
    gravityScale=0.35
    simulationSpace=RootRelative
    boundBone {boneName=dyn_Strap}
    componentAppliedLinearAccClamp=<3000 3000 3000>
    componentLinearAccScale=<1 1 1>
    componentLinearVelScale=<1 1 1>
    gravityOverride {z=-980}
    physicsBodyDefinitions [{
      boundBone {boneName=dyn_Strap}
      boxExtents=<20 30 30>
      constraintSetup {angularLimitsMax=<0 30 25>, angularLimitsMin=<0 0 -25>, angularTarget=<1 0 0>}
      localJointOffset=<4 0 0>
     }]
    relativeSpaceBone {boneName=thigh_r}
   }
   n4 = Anim Dynamics - Bone: dyn_Can!(ComponentPose=@n3, Alpha=1, GravityOverride:Vector="0,0,-980", ExternalForce="0,0,0") -> (Pose) {
    angularDampingOverride=1
    angularSpringConstant=1000
    bAngularSpring=true
    bOverrideAngularDamping=true
    bUseGravityOverride=true
    gravityScale=0.35
    simulationSpace=RootRelative
    boundBone {boneName=dyn_Can}
    componentAppliedLinearAccClamp=<3000 3000 3000>
    componentLinearAccScale=<1 1 1>
    componentLinearVelScale=<1 1 1>
    gravityOverride {z=-980}
    physicsBodyDefinitions [{
      boundBone {boneName=dyn_Can}
      boxExtents=<20 30 30>
      constraintSetup {angularLimitsMax=<0 35 15>, angularLimitsMin=<0 7 -10>, angularTarget=<1 0 0>}
      localJointOffset=<6 0 0>
     }]
   }
   n5 = "Transform (Modify) Bone - Bone: Base_Strap"(ComponentPose=@n4, Alpha=1, Translation="-3.200,0.000,0.574", Rotation="-4.928,0.000,-0.000", Scale="1,1,1") {
    rotationMode=BMM_Additive
    translationMode=BMM_Additive
    boneToModify {boneName=Base_Strap}
    rotation {pitch=-4.928}
    translation {x=-3.2, z=0.574}
   }
   Output Pose(Result=Component To Local(ComponentPose=@n5)) {layerGroup=None}
  ]}
 EventGraph {nodes [
   Event BlueprintUpdateAnimation()
   > n1 = "Gravity Override State Function (Nativized)"!(
    execute
    CharacterPartAnimInstance:Custom Character Part Anim Instance Object Reference=Self-Reference()
    GravityOverrideParameterSet:Fort Gravity Override Parameter Set Structure=Get Gravity Override Params() -> (Gravity Override Params:Fort Gravity Override Parameter Set Structure)
    RigidBodyAnimNodeParameterSet:Fort Rigid Body Anim Node Parameter Set Structure=Get RBAN Params() -> (RBAN Params:Fort Rigid Body Anim Node Parameter Set Structure)
    DebugDraw:bool=false
    GameTimeInSeconds:float=GetGameTimeInSeconds()
    DebugJointName:Name=root
    StateLogicLayerName:Name=none
   ) -> (
    then
    OutGravityOverride:Vector
    OutCurrentState:Name=None
    OutGravityOverrideParameters:Fort Gravity Override Parameters Structure
    OutRigidBodyAnimNodeParameters:Fort Rigid Body Anim Node Parameters Structure
   )
   > Set GravityOverride(GravityOverride=@n1)
   > "Set SimSpace Settings "("SimSpace Settings ":Fort Rigid Body Anim Node Parameters Structure) -> (Output_Get:Fort Rigid Body Anim Node Parameters Structure)
   Event OnCharacterCustomizationLoaded()
   > n2 = Cast To FortPlayerPawn(Object=GetOwningActor())
   > n3 = GetVariantChannelActiveIndexForItemActor!(
    execute
    Item:Athena Cosmetic Item Definition Object Reference=/BRCosmetics/Athena/Items/Cosmetics/Characters/Character_FoundKelp_Bolt
    Actor:Actor Object Reference=@n2
    Channel:Gameplay Tag Structure=(TagName="Cosmetics.Variant.Channel.Particle")
    VariantList:Array of Gameplay Tag Structures=Make Array!("[0]"=(TagName="Cosmetics.Variant.Property.Stage1"), "[1]":Gameplay Tag Structure=(TagName="Cosmetics.Variant.Property.Stage2")) -> (Array)
    bErrorIfChannelMissing:bool=false
    bErrorIfTagNotInArray:bool=false
   ) -> (then, Index:int=0, bIndexIsValid:bool, bFoundChannel:bool)
   > n4 = Branch(Condition="Equal (Integer)"(A=@n3))
   > Set isUsingHelmet(isUsingHelmet:bool=false) -> (Output_Get:bool)
   Set isUsingHelmet(execute=@n4.else, isUsingHelmet=true)
  ]}
]
variables [
 Gravity Override Params {
  type=FortGravityOverrideParameterSet
  default=(Skydive_GoingUp=(GravityOverride=()),Skydive_FloatingDown=(GravityOverride=()),Skydive_DivingDown=(GravityOverride=()),Skydive_Parachuting=(GravityOverride=()),OnGround_Standing=(GravityOverride=()),OnGround_Moving=(GravityOverride=()),OnGround_Moving_HighRotationRate=(GravityOverride=()),OnGround_Crouch_Moving=(GravityOverride=()),EmoteOrMelee=(GravityOverride=()),Swimming=(GravityOverride=()),Front_End=(GravityOverride=()),DBNO=(GravityOverride=()),DBNO_Carried=(GravityOverride=()),OnGround_Falling=(GravityOverride=()),Ziplining=(GravityOverride=()),Skydive_Surfing=(GravityOverride=()),Riding_Standing=(GravityOverride=()),Riding_Moving=(GravityOverride=()),Riding_Moving_Targeting=(GravityOverride=()),Tactical_Sprint=(GravityOverride=()),Rail_Grinding=(GravityOverride=()),)
 }
 GravityOverride {type=Vector}
 isUsingHelmet {type=boolean}
 RBAN Params {
  type=FortRigidBodyAnimNodeParameterSet
  default=(Skydive_GoingUp=(RigidBodyAnimNode=()),Skydive_FloatingDown=(RigidBodyAnimNode=()),Skydive_DivingDown=(RigidBodyAnimNode=()),Skydive_Parachuting=(RigidBodyAnimNode=()),OnGround_Standing=(RigidBodyAnimNode=()),OnGround_Moving=(RigidBodyAnimNode=()),OnGround_Moving_HighRotationRate=(RigidBodyAnimNode=()),OnGround_Crouch_Moving=(RigidBodyAnimNode=()),EmoteOrMelee=(RigidBodyAnimNode=()),Swimming=(RigidBodyAnimNode=()),Front_End=(RigidBodyAnimNode=()),DBNO=(RigidBodyAnimNode=()),DBNO_Carried=(RigidBodyAnimNode=()),OnGround_Falling=(RigidBodyAnimNode=()),Ziplining=(RigidBodyAnimNode=()),Skydive_Surfing=(RigidBodyAnimNode=()),Riding_Standing=(RigidBodyAnimNode=()),Riding_Moving=(RigidBodyAnimNode=()),Riding_Moving_Targeting=(RigidBodyAnimNode=()),Tactical_Sprint=(RigidBodyAnimNode=()),Rail_Grinding=(RigidBodyAnimNode=()),)
 }
 "SimSpace Settings " {
  type=FortRigidBodyAnimNodeParameters
  default {simSpaceSettings {maxAngularAcceleration=1, maxAngularVelocity=3, maxLinearAcceleration=800, maxLinearVelocity=500, worldAlpha=1, externalLinearDragV=<1 1 1>}}
 }
]
